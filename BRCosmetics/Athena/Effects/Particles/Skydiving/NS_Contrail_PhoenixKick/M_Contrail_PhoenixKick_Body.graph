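# Masked, lit material for the body mesh particles of the PhoenixKick skydiving contrail Niagara effect.
# A time-driven sine wave along UV1 builds local-axis offsets, tapered toward both ends to give a wavy motion.
# Dynamic parameters set wave tiling, speed, per-particle randomness and X/Z strength.
# Surface nodes include contrast-adjusted diffuse, a flatten-adjusted normal, packed mask channels and a dithered near-camera fade.
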Material /BRCosmetics/Athena/Effects/Particles/Skydiving/NS_Contrail_PhoenixKick/M_Contrail_PhoenixKick_Body
graph {
 properties {bAutomaticallySetUsageInEditor=false, bUsedWithNiagaraMeshParticles=true, blendMode=BLEND_Masked, usedShadingModels=Default Lit}
 nodes [
  n1 = Transform(None=Constant3Vector() {constant=<rgba 1 0 0 0>}) {transformSourceType=TRANSFORMSOURCE_Local}
  n2 = DynamicParameter() {defaultValue=<rgba 0 0 1 1>, paramNames [UVTilingRandom, TimeOffsetRandom, WPOUVTiling, WPOspeedoffset]}
  n3 = Add(A=Time(), B=@n2.TimeOffsetRandom)
  n4 = Multiply(A=@n2.WPOspeedoffset) {constB=-2}
  n5 = Add(A=@n2.UVTilingRandom, B=ComponentMask(None=TextureCoordinate() {coordinateIndex=1}) {r=true})
  n6 = Multiply(A=@n2.WPOUVTiling) {constB=2}
  n7 = Sine(None=Add(A=Multiply(A=@n3, B=@n4) {constB=-2}, B=Multiply(A=@n5, B=@n6) {constB=2}))
  n8 = OneMinus(None=ComponentMask(None=TextureCoordinate() {coordinateIndex=1}) {r=true})
  n9 = Multiply(A=@n8) {constB=0.65}
  n10 = Multiply(A=OneMinus(None=@n8)) {constB=3}
  n11 = Multiply(
   A=MaterialFunctionCall(X=@n7, Y=@n7, Z=@n7) {materialFunction=&MakeFloat3}.Result
   B=Multiply(A=Power(Base=@n9) {constExponent=1.5}, B=Power(Base=@n10) {constExponent=1.5})
  )
  n12 = Constant() {r=60}
  n13 = DynamicParameter() {parameterIndex=1, paramNames [WPO_X_Int,WPO_Z_Int,Param3,Param4]}
  n14 = Transform(None=Constant3Vector() {constant=<rgba 0 0 1 0>}) {transformSourceType=TRANSFORMSOURCE_Local}
  n15 = Constant() {r=60}
  n16 = Add(
   A=Multiply(A=Multiply(A=@n1, B=@n11), B=Multiply(A=@n12, B=@n13.WPO_X_Int)) {constB=60}
   B=Multiply(A=Multiply(A=@n14, B=@n11), B=Multiply(A=@n15, B=@n13.WPO_Z_Int)) {constB=60}
  )
  n17 = Transform(None=Constant3Vector() {constant=<rgba 0 1 0 0>}) {transformSourceType=TRANSFORMSOURCE_Local}
  Add(A=@n16, B=Multiply(A=Multiply(A=@n17, B=@n11)) {constB=20})
  Constant()
  Constant() {r=0.5}
  MaterialFunctionCall(Alpha Threshold=MaterialFunctionCall(Distance=ScalarParameter() {parameterName=CameraDepthFade}) {materialFunction=&Near_Camera_Fade}.Result) {materialFunction=&DitherTemporalAA}
  n18 = TextureSample() {samplerType=SAMPLERTYPE_Normal, texture=&/T_Contrail_PhoenixKick_N}
  MaterialFunctionCall(Normal=@n18.RGB, Flatness=Constant() {r=-1.5}) {materialFunction=&FlattenNormal}
  n19 = TextureSample() {texture=&/T_Contrail_PhoenixKick_D}
  Multiply(
   A=@n19.RGB
   B=TextureSample() {texture=&/T_Contrail_PhoenixKick_BodyMask}.R
  )
  Reroute(None=MaterialFunctionCall!(In=@n19.RGB, Contrast=ScalarParameter() {parameterName=Contrast}) {materialFunction=/Engine/Functions/Engine_MaterialFunctions01/ImageAdjustment/CheapContrast_RGB}.Result)
  Reroute(None=@n16)
  n20 = TextureSample() {samplerType=SAMPLERTYPE_Masks, texture=&/T_Contrail_PhoenixKick_S}
  Reroute(None=@n20.R)
  Reroute(None=LinearInterpolate(Alpha=@n20.G) {constB=0.5})
 ]
}
