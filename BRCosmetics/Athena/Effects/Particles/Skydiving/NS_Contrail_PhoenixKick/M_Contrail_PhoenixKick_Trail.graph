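# Translucent unlit Niagara mesh-particle material for the Phoenix Kick skydiving contrail trail.
# World position offset combines panned-mask normal displacement with a UV-masked sine wave randomized by dynamic parameters.
# Emissive multiplies a UV gradient by noisy Fresnel colors plus offset-sampled chromatic caustics, zero-weighted by default.
# Opacity combines panned noise, ramp mask, particle alpha and optional height-lerp dissolve, softened by Fresnel, camera, view-angle and depth fades.
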Material /BRCosmetics/Athena/Effects/Particles/Skydiving/NS_Contrail_PhoenixKick/M_Contrail_PhoenixKick_Trail
graph {
 properties {
  bAutomaticallySetUsageInEditor=false
  bUsedWithNiagaraMeshParticles=true
  blendMode=BLEND_Translucent
  shadingModel=MSM_Unlit
  usedShadingModels=Unlit
 }
 nodes [
  n1 = TextureCoordinate() {vTiling=3}
  n2 = Multiply(A=Time(), B=ScalarParameter() {defaultValue=1, parameterName=WPO_Speed})
  n3 = Multiply(
   A=TextureSample(UVs=Panner(Coordinate=@n1, Time=@n2) {speedX=-1, speedY=3}) {texture=&/T_Contrail_PhoenixKick_WPOMask}.R
   B=ScalarParameter() {defaultValue=5, parameterName=WPO_Int}
  ) {constB=25}
  n4 = ComponentMask(None=TextureCoordinate()) {g=true}
  n5 = Saturate(None=Power(Base=Multiply(A=@n4) {constB=5}))
  n6 = Constant3Vector() {constant=<rgba 1 0 0 0>}
  n7 = DynamicParameter() {defaultValue=<rgba 0 0 1 1>, paramNames [UVTilingRandom, TimeOffsetRandom, WPOUVTiling, WPOspeedoffset]}
  n8 = Add(A=Time(), B=@n7.TimeOffsetRandom)
  n9 = Multiply(A=@n7.WPOspeedoffset) {constB=2}
  n10 = Add(A=@n7.UVTilingRandom, B=ComponentMask(None=TextureCoordinate() {coordinateIndex=1}) {g=true})
  n11 = Multiply(A=@n7.WPOUVTiling) {constB=2}
  n12 = Sine(None=Add(A=Multiply(A=@n8, B=@n9) {constB=2}, B=Multiply(A=@n10, B=@n11) {constB=2}))
  n13 = OneMinus(None=ComponentMask(None=TextureCoordinate() {coordinateIndex=1}) {g=true})
  n14 = Multiply(A=OneMinus(None=@n13)) {constB=3}
  n15 = Multiply(
   A=MaterialFunctionCall(X=@n12, Y=@n12, Z=@n12) {materialFunction=&MakeFloat3}.Result
   B=Multiply(A=Multiply(A=@n13), B=Power(Base=@n14) {constExponent=1.5})
  )
  n16 = Multiply(
   A=Multiply(A=Transform(None=@n6) {transformSourceType=TRANSFORMSOURCE_Local}, B=@n15)
   B=ScalarParameter() {defaultValue=2, parameterName=LinearWPO_X_Int}
  ) {constB=2}
  n17 = Constant3Vector() {constant=<rgba 0 1 0 0>}
  n18 = Multiply(A=Multiply(A=Transform(None=@n17) {transformSourceType=TRANSFORMSOURCE_Local}, B=@n15), B=ScalarParameter() {parameterName=LinearWPO_Z_Int}) {constB=0}
  Add(
   A=Multiply(A=Multiply(A=@n3, B=@n5), B=VertexNormalWS()) {constB=0}
   B=Multiply(A=Add(A=@n16, B=@n18), B=ScalarParameter() {defaultValue=20, parameterName=WPO_Linearlnt}) {constB=20}
  )
  n19 = ParticleColor()
  n20 = TextureCoordinate()
  n21 = VectorParameter() {parameterName=BaseTextureUV, defaultValue=<rgba 1 4 -1 0>}
  n22 = AppendVector(A=@n21.R, B=@n21.R)
  n23 = TextureSampleParameter2D(UVs=Panner(Coordinate=Multiply(A=@n20, B=@n22), Speed=AppendVector(A=@n21.B, B=@n21.B)) {speedX=-1, speedY=1}) {
   parameterName=BaseTexture
   samplerType=SAMPLERTYPE_LinearGrayscale
   texture=&T_SimplePerlin_Greyscale
  }
  n24 = ScalarParameter() {defaultValue=1.3, parameterName=BaseTexture_Power}
  n25 = TextureCoordinate() {uTiling=0.5}
  n26 = Multiply(
   A=Multiply(A=@n19.A, B=Power(Base=@n23.R, Exp=@n24) {constExponent=1.2})
   B=TextureSampleParameter2D(UVs=Panner(Coordinate=@n25) {speedX=-1}) {parameterName=MaskTexture, texture=/Game/Effects/Fort_Effects/Textures/Ramps/T_Power_Source_Sphere_Ramp}.R
  )
  n27 = VectorParameter() {parameterName=DissolveUV, defaultValue=<rgba 1 1 -1 0>}
  n28 = Multiply(A=TextureCoordinate(), B=AppendVector(A=@n27.R, B=@n27.R))
  n29 = AppendVector(A=@n27.B, B=@n27.B)
  n30 = MaterialFunctionCall(
   Transition Phase=LinearInterpolate(
    A=ScalarParameter() {defaultValue=0.1, parameterName=DissMin}
    B=ScalarParameter() {defaultValue=1, parameterName=DissMax}
    Alpha=DynamicParameter() {parameterIndex=1, defaultValue=<rgba 0 1 1 1>, paramNames [Dissolve,Param2,Param3,Param4]}.Dissolve
   ) {constA=0.3, constB=0.8}
   Height Texture=TextureSampleParameter2D(UVs=Panner(Coordinate=@n28, Speed=@n29)) {parameterName=DissolveTexture, texture=/BRCosmetics/Weapons/FORT_Melee/Pickaxe_SoapPocket/FX/T_Pickaxe_SoapPocket_Noise}.R
   Contrast=ScalarParameter() {defaultValue=2, parameterName="Contrast (S)"}
  ) {materialFunction=&HeightLerp}
  n31 = Fresnel() {baseReflectFraction=0}
  n32 = Multiply(
   A=LinearInterpolate(A=@n26, B=Multiply(A=@n26, B=@n30.Results), Alpha=ScalarParameter() {parameterName=Use_Dissolve})
   B=Power(Base=OneMinus(None=@n31), Exp=ScalarParameter() {defaultValue=2, parameterName=MaskFresnel_Power}) {constExponent=5}
  )
  n33 = MaterialFunctionCall(Distance=ScalarParameter() {defaultValue=100, parameterName=NearCameraFade}) {materialFunction=&Near_Camera_Fade}
  n34 = ScalarParameter() {parameterName=MeshFade}
  DepthFade(
   Opacity=Multiply(
    A=Multiply(A=@n32, B=@n33.Result)
    B=MaterialFunctionCall("FadeSensitivity (optional)"=@n34) {materialFunction=&FadeBasedOnViewAngle}.Result
   )
   FadeDistance=ScalarParameter() {defaultValue=5, parameterName=Depthfade}
  )
  n35 = TextureCoordinate()
  n36 = LinearInterpolate(
   A=VectorParameter() {parameterName=LinearColorA, defaultValue=<rgba 0 0.429 1 1>}.RGB
   B=VectorParameter() {parameterName=LinearColorB, defaultValue=<rgba 0.0677 1 0.45 1>}.RGB
   Alpha=Power(Base=ComponentMask(None=@n35) {g=true}, Exp=ScalarParameter() {defaultValue=1.2, parameterName=LinearColor_MaskPower}) {constExponent=1.2}
  )
  n37 = ScalarParameter() {defaultValue=-0.01, parameterName=ColorNoiseA_Min}
  n38 = Add(A=TextureCoordinate() {vTiling=3})
  n39 = ScalarParameter() {defaultValue=0.05, parameterName=Offset}
  n40 = ComponentMask(None=VectorParameter() {parameterName=ColorNoiseASpeed, defaultValue=<rgba -1 1 0 0>}.RGB) {g=true, r=true}
  n41 = TextureSampleParameter2D(UVs=Panner(Coordinate=Add(A=@n38, B=@n39), Speed=@n40) {speedX=0.5, speedY=1}) {parameterName=ColorNoiseA, texture=/Game/Effects/Fort_Effects/Textures/Tile/T_ShieldCaustics_Tile_Chromatic}
  n42 = TextureCoordinate() {vTiling=3}
  n43 = TextureSampleParameter2D(UVs=Panner(Coordinate=Add(A=@n42), Speed=@n40) {speedX=0.5, speedY=1}) {parameterName=ColorNoiseA, texture=&T_ShieldCaustics_Tile_Chromatic}
  n44 = Add(A=TextureCoordinate() {vTiling=3})
  n45 = TextureSampleParameter2D(UVs=Panner(Coordinate=Subtract(A=@n44, B=@n39), Speed=@n40) {speedX=0.5, speedY=1}) {parameterName=ColorNoiseA, texture=&T_ShieldCaustics_Tile_Chromatic}
  n46 = Desaturation(
   None=MaterialFunctionCall(X=LinearInterpolate(A=@n37, Alpha=@n41.R), Y=LinearInterpolate(A=@n37, Alpha=@n43.R), Z=LinearInterpolate(A=@n37, Alpha=@n45.R)) {materialFunction=&MakeFloat3}.Result
   Fraction=ScalarParameter() {parameterName=Des}
  )
  n47 = ScalarParameter() {defaultValue=1.5, parameterName=ColorPower}
  n48 = VectorParameter() {parameterName=FresnelColorA, defaultValue=<rgba 0.0099 0.0971 0.1 1>}
  n49 = VectorParameter() {parameterName=FresnelColorB, defaultValue=<rgba 0.0923 1.2 0.7762 1>}
  n50 = OneMinus(None=Fresnel() {baseReflectFraction=0, exponent=1})
  n51 = ScalarParameter() {defaultValue=6, parameterName=Fresnel_Power}
  n52 = VectorParameter() {parameterName=FresnelMask_TextureUV, defaultValue=<rgba 1 3 -1 0>}
  n53 = Panner(Coordinate=Multiply(A=TextureCoordinate(), B=AppendVector(A=@n52.R, B=@n52.R)), Speed=AppendVector(A=@n52.B, B=@n52.B)) {speedX=-1}
  n54 = Saturate(None=Multiply(
    A=Power(Base=@n50, Exp=@n51) {constExponent=6}
    B=TextureSample(UVs=@n53) {samplerType=SAMPLERTYPE_LinearGrayscale, texture=/Game/Effects/Fort_Effects/Emotes/UltraLight/Texture/T_DirectionalNoise_Greyscale}.R
   ))
  n55 = Add(
   A=Multiply(A=Power(Base=@n46, Exp=@n47) {constExponent=1}, B=ScalarParameter() {parameterName=ColorTexture_Mul}) {constB=2}
   B=Multiply(A=LinearInterpolate(A=@n48.RGB, B=@n49.RGB, Alpha=@n54), B=@n19.RGB)
  )
  MaterialFunctionCall(Emissive=Desaturation(None=Multiply(A=@n36, B=@n55), Fraction=ScalarParameter() {parameterName=Desaturation})) {materialFunction=&MF_FX_EmissiveAdjustments}
 ]
}
